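# Physics asset for the Heart Breaker backpack cosmetic, defining secondary-motion bodies for its bag flap, grenade and nine tassels.
# Each tassel chains three capsules from a box base, while flap and grenade are single capsules jointed to box bases.
# Angular limits keep twist very tight, with tassels 01-05 stiffer at the root than 06-09.
# A capsule covers the pack body and has no constraints.
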PhysicsAsset /BRCosmetics/Accessories/FORT_Backpacks/Backpack_F_MED_Heart_Breaker/Meshes/F_MED_Heart_Breaker_Pack_Physics
bodies {
 dyn_bag_flap_base [{shapeType=Box, extentX=0.5, extentY=0.5, extentZ=0.5}]
 dyn_bag_flap_parent [{shapeType=Capsule, length=15, radius=4, center=<6.6823 0 1.4247>, rotation=<rot 79.9999 -179.9999 -180>}]
 dyn_grenade_base [{shapeType=Box, extentX=0.5, extentY=0.5, extentZ=0.5}]
 dyn_grenade_parent [{shapeType=Capsule, length=2.9684, radius=2.8952, center=<1.8392 0 0>, rotation=<rot 90 0 0>}]
 dyn_tassel_01_a [{shapeType=Capsule, length=2, radius=1.2, center=<2.3744 0 0>, rotation=<rot 90 0 0>}]
 dyn_tassel_01_b [{shapeType=Capsule, length=1, radius=1.2, center=<1.5 0 0>, rotation=<rot 90 0 0>}]
 dyn_tassel_01_base [{shapeType=Box, extentX=1.25, extentY=1.25, extentZ=1.25}]
 dyn_tassel_01_parent [{shapeType=Capsule, length=3.5, radius=1.2, center=<3.0516 0 0>, rotation=<rot 90 0 0>}]
 dyn_tassel_02_a [{shapeType=Capsule, length=2.5, radius=1.2, center=<2.3744 0 0>, rotation=<rot 90 0 0>}]
 dyn_tassel_02_b [{shapeType=Capsule, length=1, radius=1.2, center=<1.5 0 0>, rotation=<rot 90 0 0>}]
 dyn_tassel_02_base [{shapeType=Box, extentX=1.25, extentY=1.25, extentZ=1.25}]
 dyn_tassel_02_parent [{shapeType=Capsule, length=3.5, radius=1.2, center=<3.0516 0 0>, rotation=<rot 90 0 0>}]
 dyn_tassel_03_a [{shapeType=Capsule, length=3, radius=1.2, center=<2.5301 0 0>, rotation=<rot 90 0 0>}]
 dyn_tassel_03_b [{shapeType=Capsule, length=1, radius=1.2, center=<1.5 0 0>, rotation=<rot 90 0 0>}]
 dyn_tassel_03_base [{shapeType=Box, extentX=1.25, extentY=1.25, extentZ=1.25}]
 dyn_tassel_03_parent [{shapeType=Capsule, length=1.5, radius=1.2, center=<2 0 0>, rotation=<rot 90 0 0>}]
 dyn_tassel_04_a [{shapeType=Capsule, length=3.5, radius=1.2, center=<2.6746 0 0>, rotation=<rot 90 0 0>}]
 dyn_tassel_04_b [{shapeType=Capsule, length=1.5, radius=1.2, center=<1.8041 0 0>, rotation=<rot 90 0 0>}]
 dyn_tassel_04_base [{shapeType=Box, extentX=1.25, extentY=1.25, extentZ=1.25}]
 dyn_tassel_04_parent [{shapeType=Capsule, length=1.5, radius=1.2, center=<2 0 0>, rotation=<rot 90 0 0>}]
 dyn_tassel_05_a [{shapeType=Capsule, length=3.5, radius=1.2, center=<2.8117 0 0>, rotation=<rot 90 0 0>}]
 dyn_tassel_05_b [{shapeType=Capsule, length=2.8, radius=1.2, center=<2.3324 0 0>, rotation=<rot 90 0 0>}]
 dyn_tassel_05_base [{shapeType=Box, extentX=1.25, extentY=1.25, extentZ=1.25}]
 dyn_tassel_05_parent [{shapeType=Capsule, length=2, radius=1.2, center=<2 0 0>, rotation=<rot 90 0 0>}]
 dyn_tassel_06_a [{shapeType=Capsule, length=3.5, radius=1.2, center=<3 0 0>, rotation=<rot 90 0 0>}]
 dyn_tassel_06_b [{shapeType=Capsule, length=1, radius=1.2, center=<2 0 0>, rotation=<rot 90 0 0>}]
 dyn_tassel_06_base [{shapeType=Box, extentX=1.25, extentY=1.25, extentZ=1.25}]
 dyn_tassel_06_parent [{shapeType=Capsule, length=1.2, radius=1.2, center=<2 0 0>, rotation=<rot 90 0 0>}]
 dyn_tassel_07_a [{shapeType=Capsule, length=3.5, radius=1.2, center=<3 0 0>, rotation=<rot 90 0 0>}]
 dyn_tassel_07_b [{shapeType=Capsule, length=1, radius=1.2, center=<2 0 0>, rotation=<rot 90 0 0>}]
 dyn_tassel_07_base [{shapeType=Box, extentX=1.25, extentY=1.25, extentZ=1.25}]
 dyn_tassel_07_parent [{shapeType=Capsule, length=1.5, radius=1.2, center=<2 0 0>, rotation=<rot 90 0 0>}]
 dyn_tassel_08_a [{shapeType=Capsule, length=3.5, radius=1.2, center=<3 0 0>, rotation=<rot 90 0 0>}]
 dyn_tassel_08_b [{shapeType=Capsule, length=0.7, radius=1.2, center=<2 0 0>, rotation=<rot 90 0 0>}]
 dyn_tassel_08_base [{shapeType=Box, extentX=1.25, extentY=1.25, extentZ=1.25}]
 dyn_tassel_08_parent [{shapeType=Capsule, length=1.6, radius=1.2, center=<2 0 0>, rotation=<rot 90 0 0>}]
 dyn_tassel_09_a [{shapeType=Capsule, length=1.5, radius=1.2, center=<2 0 0>, rotation=<rot 90 0 0>}]
 dyn_tassel_09_b [{shapeType=Capsule, length=2, radius=1.2, shapeName=dyn_tassel_09_b_capsule, center=<2 0 0>, rotation=<rot 90 0 0>}]
 dyn_tassel_09_base [{shapeType=Box, extentX=1.25, extentY=1.25, extentZ=1.25}]
 dyn_tassel_09_parent [{shapeType=Capsule, length=1.8, radius=1.2, center=<2 0 0>, rotation=<rot 90 0 0>}]
 pack [{shapeType=Capsule, length=18, radius=4.5, shapeName=pack_capsule, center=<-0.7877 -3.7944 -11.8581>, rotation=<rot -70 0.0004 0>}]
}
constraints [
 {bone1Name=dyn_tassel_06_b, bone2Name=dyn_tassel_06_a, swing1LimitDegrees=10, swing2LimitDegrees=20, twistLimitDegrees=3}
 {bone1Name=dyn_tassel_06_a, bone2Name=dyn_tassel_06_parent, swing1LimitDegrees=20, swing2LimitDegrees=30, twistLimitDegrees=2}
 {bone1Name=dyn_tassel_09_b, bone2Name=dyn_tassel_09_a, swing1LimitDegrees=10, swing2LimitDegrees=20, twistLimitDegrees=3}
 {bone1Name=dyn_tassel_09_a, bone2Name=dyn_tassel_09_parent, swing1LimitDegrees=20, swing2LimitDegrees=30, twistLimitDegrees=2}
 {bone1Name=dyn_tassel_09_parent, bone2Name=dyn_tassel_09_base, swing1LimitDegrees=20, swing2LimitDegrees=30, twistLimitDegrees=1}
 {bone1Name=dyn_tassel_08_b, bone2Name=dyn_tassel_08_a, swing1LimitDegrees=10, swing2LimitDegrees=20, twistLimitDegrees=3}
 {bone1Name=dyn_tassel_08_a, bone2Name=dyn_tassel_08_parent, swing1LimitDegrees=20, swing2LimitDegrees=30, twistLimitDegrees=2}
 {bone1Name=dyn_tassel_08_parent, bone2Name=dyn_tassel_08_base, swing1LimitDegrees=20, swing2LimitDegrees=30, twistLimitDegrees=1}
 {bone1Name=dyn_tassel_07_b, bone2Name=dyn_tassel_07_a, swing1LimitDegrees=10, swing2LimitDegrees=20, twistLimitDegrees=3}
 {bone1Name=dyn_tassel_07_a, bone2Name=dyn_tassel_07_parent, swing1LimitDegrees=20, swing2LimitDegrees=30, twistLimitDegrees=2}
 {bone1Name=dyn_tassel_07_parent, bone2Name=dyn_tassel_07_base, swing1LimitDegrees=20, swing2LimitDegrees=30, twistLimitDegrees=1}
 {bone1Name=dyn_tassel_06_parent, bone2Name=dyn_tassel_06_base, swing1LimitDegrees=15, swing2LimitDegrees=25, twistLimitDegrees=1}
 {bone1Name=dyn_tassel_05_b, bone2Name=dyn_tassel_05_a, swing1LimitDegrees=10, swing2LimitDegrees=20, twistLimitDegrees=3}
 {bone1Name=dyn_tassel_05_a, bone2Name=dyn_tassel_05_parent, swing1LimitDegrees=20, swing2LimitDegrees=30, twistLimitDegrees=2}
 {bone1Name=dyn_tassel_05_parent, bone2Name=dyn_tassel_05_base, swing1LimitDegrees=5, swing2LimitDegrees=5, twistLimitDegrees=1}
 {bone1Name=dyn_tassel_04_b, bone2Name=dyn_tassel_04_a, swing1LimitDegrees=10, swing2LimitDegrees=20, twistLimitDegrees=3}
 {bone1Name=dyn_tassel_04_a, bone2Name=dyn_tassel_04_parent, swing1LimitDegrees=20, swing2LimitDegrees=30, twistLimitDegrees=2}
 {bone1Name=dyn_tassel_04_parent, bone2Name=dyn_tassel_04_base, swing1LimitDegrees=5, swing2LimitDegrees=5, twistLimitDegrees=1}
 {bone1Name=dyn_tassel_03_b, bone2Name=dyn_tassel_03_a, swing1LimitDegrees=10, swing2LimitDegrees=20, twistLimitDegrees=3}
 {bone1Name=dyn_tassel_03_a, bone2Name=dyn_tassel_03_parent, swing1LimitDegrees=20, swing2LimitDegrees=30, twistLimitDegrees=2}
 {bone1Name=dyn_tassel_03_parent, bone2Name=dyn_tassel_03_base, swing1LimitDegrees=5, swing2LimitDegrees=5, twistLimitDegrees=1}
 {bone1Name=dyn_tassel_02_b, bone2Name=dyn_tassel_02_a, swing1LimitDegrees=10, swing2LimitDegrees=20, twistLimitDegrees=3}
 {bone1Name=dyn_tassel_02_a, bone2Name=dyn_tassel_02_parent, swing1LimitDegrees=20, swing2LimitDegrees=30, twistLimitDegrees=2}
 {bone1Name=dyn_tassel_02_parent, bone2Name=dyn_tassel_02_base, swing1LimitDegrees=5, swing2LimitDegrees=5, twistLimitDegrees=1}
 {bone1Name=dyn_tassel_01_b, bone2Name=dyn_tassel_01_a, swing1LimitDegrees=10, swing2LimitDegrees=20, twistLimitDegrees=3}
 {bone1Name=dyn_tassel_01_a, bone2Name=dyn_tassel_01_parent, swing1LimitDegrees=20, swing2LimitDegrees=30, twistLimitDegrees=2}
 {bone1Name=dyn_tassel_01_parent, bone2Name=dyn_tassel_01_base, swing1LimitDegrees=5, swing2LimitDegrees=5, twistLimitDegrees=1}
 {bone1Name=dyn_bag_flap_parent, bone2Name=dyn_bag_flap_base, swing1LimitDegrees=5, swing2LimitDegrees=15, twistLimitDegrees=1}
 {bone1Name=dyn_grenade_parent, bone2Name=dyn_grenade_base, swing1LimitDegrees=10, swing2LimitDegrees=10, twistLimitDegrees=3}
]
